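# Niagara cosmetic effect for the BriskReign character body, emitting mist and glowing embers from four skeletal mesh bone/socket points, apparently hands and feet.
# Particles are partitioned per point and killed by the matching hand-hidden or foot-hidden user flags.
# Embers spawn there, drift with curl noise and drag, and fade as owner speed rises.
# Mist re-samples its point each update, animates sub-images, and adapts to quality level, front end and time of day.
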NiagaraSystem /BRCosmetics/Characters/Player/Male/Medium/Bodies/M_MED_BriskReign/FX/NS_BriskReign_BodyMist
dependencies {
 usedDataInterfaces [&NiagaraDataInterfaceCurve, &NiagaraDataInterfaceColorCurve, &NiagaraDataInterfaceSkeletalMesh]
 usedDynamicInputs [
  &Vector2DFromFloat
  &MakeVector4_FromLinearColor
  &RotateVectorByQuaternion
  &ColorFromCurve
  &Vector4FromCurve
 ]
 usedModules [
  &SystemState
  &self:SystemSpawnScript.SystemScriptSource.SystemScriptGraph.NiagaraNodeAssignment_0.SetVariables_#1
 ]
}
emitters {
 Embers {
  emitterUpdateScript [
   &EmitterState()
   &SpawnRate(SpawnRate=80, Use Spawn Probability=false)
  ]
  particleSpawnScript [
   /SetVariables(Particles.PerParticleRandom:NiagaraFloat=&RandomRangeFloat(Minimum=0, Maximum=1)) {isSetParameters=true}
   &PartitionParticles(Number Of Partitions=4)
   &InitializeParticle(
    Color=&MultiplyLinearColor(Linear Color=<rgba 1 1 1 0.8>, Scale Factor=&MakeVector4_FromFloat(Float=1))
    Lifetime=&RandomRangeFloat(Minimum=0.6, Maximum=1)
    Sprite Size=&Multiply_Vector2D(
     A=<0.4 0.4>
     B=&MakeVector2D(
      X=&FloatFromCurve(
       FloatCurve=DataInterface() {propertyValues {Curve {keys [<at 0 1>,<at 0.8 2 RCIM_Constant>,<at 0.95 3 RCIM_Constant>]}}}
       CurveIndex=$Particles.PerParticleRandom
      )
      Y=&FloatFromCurve(
       FloatCurve=DataInterface() {propertyValues {Curve {keys [<at 0 1>,<at 0.8 2 RCIM_Constant>,<at 0.95 3>]}}}
       CurveIndex=$Particles.PerParticleRandom
      )
     )
    )
    Sprite Size Mode=::NewEnumerator4
    UsePositionOffset=false
    Write Material Random=true
   )
   &SkeletalMeshLocation(
    Bone / Socket Index=$Particles.PartitionParticles.Partition
    Bone Sampling Mode=::NewEnumerator4
    "Inherit Velocity (Scale)"=0
    MaskSpawnGroup=false
    Sampled Position Offset=&MergeHelper_VectorSwitch(
     0=&RandomRangeVector(Minimum=<0 0 0>, Maximum=<25 0 0>)
     1=&RandomRangeVector(Minimum=<0 0 0>, Maximum=<-25 0 0>)
     2=&RandomRangeVector(Minimum=<0 0 0>, Maximum=<30 0 5>)
     3=&RandomRangeVector(Minimum=<0 0 0>, Maximum=<-30 -6 -5>)
     Index=$Particles.PartitionParticles.Partition
     Num=::NewEnumerator2
    )
    Skeletal Mesh=$System.Skeletal Mesh Reader
    Velocity Sampling=::NewEnumerator0
    VelocitySpeedLimit=false
   )
   &ShapeLocation(MaskSpawnGroup=false, Override Seed=false, Sphere Radius=10, UseOwnerScale=true)
   &AddVelocity(Velocity=&RandomRangeVector(Minimum=<-20 -20 -20>, Maximum=<20 20 20>), Velocity Speed Scale=0.6)
  ]
  particleUpdateScript [
   &ParticleState()
   &Drag(
    Drag=&FloatFromCurve(
     FloatCurve=DataInterface() {propertyValues {Curve {keys [
         {arriveTangent=0.1577, interpMode=RCIM_Cubic, leaveTangent=0.1577, tangentMode=RCTM_User, time=0, value=0.125}
         {arriveTangent=1.6704, interpMode=RCIM_Cubic, leaveTangent=1.6704, tangentMode=RCTM_User, time=1, value=1}
        ]}}}
     Scale Curve=&RandomRangeFloat(Minimum=4, Maximum=4)
    )
    Use Linear Drag=true
    Use Rotational Drag=true
   )
   &CurlNoiseForce(
    Bias Noise=false
    Noise Strength=&FloatFromCurve(
     FloatCurve=DataInterface() {propertyValues {Curve {keys [{interpMode=RCIM_Cubic, tangentMode=RCTM_User, time=0, value=0.2}, {interpMode=RCIM_Cubic, tangentMode=RCTM_User, time=1, value=1}]}}}
     Scale Curve=26
    )
    Pan Noise=false
    Random Seed=33
   ) {module=CurlNoiseForce001}
   &SolveForcesAndVelocity(Clamp Velocity=false, Limit Acceleration=false)
   &ScaleColor(
    Linear Color Curve=DataInterface() {propertyValues {
      AlphaCurve {keys [<at 0 0>,<at 0.4658 0.35 RCIM_Cubic>,<at 0.9968 0>]}
      BlueCurve {keys [<at 0.0028 5>, {arriveTangent=-4.2116, interpMode=RCIM_Cubic, leaveTangent=-4.2116, time=0.2108, value=3}, <at 1 0.8 RCIM_Cubic>]}
      GreenCurve {keys [<at 0.0028 4.401>, {arriveTangent=-3.6569, interpMode=RCIM_Cubic, leaveTangent=-3.6569, time=0.2108, value=2.7927}, <at 1 0.7542 RCIM_Cubic>]}
      RedCurve {keys [
        <at 0.0028 3.8021>
        {arriveTangent=-3.1023, interpMode=RCIM_Cubic, leaveTangent=-3.1023, time=0.2108, value=2.2969}
        <at 1 0.7083 RCIM_Cubic>
       ]}
     }}
    Scale Mode=::NewEnumerator2
   )
   &ScaleColor(
    Color Value To Scale=$Particles.Color
    Scale Alpha=&Lerp_Float(
     A=0
     B=1
     Alpha=&RemapRange(
      Input Value To Remap=&VectorLength(Vector=$Engine.Owner.Velocity)
      Input Min=0
      Input Max=200
      Output Min=1
      Output Max=0
     )
    )
    ScaleA=true
    ScaleRGB=true
   ) {module=ScaleColor002}
   &ScaleSpriteSize(
    Uniform Curve Scale=&RandomRangeFloat(Minimum=0.9, Maximum=1.1)
    Uniform Curve Sprite Scale=DataInterface() {propertyValues {Curve {keys [
        {tangentMode=RCTM_User, time=0, value=1}
        {arriveTangent=-0.0405, interpMode=RCIM_Cubic, leaveTangent=-0.0405, tangentMode=RCTM_User, time=0.7009, value=1}
        {arriveTangent=-8.0089, interpMode=RCIM_Cubic, leaveTangent=-8.0089, tangentMode=RCTM_User, time=1, value=0}
       ]}}}
   )
   /SetVariables(Particles.SpriteSizeGlow:Vector2f=&Multiply_Vector2D_ByFloat(Float=1.2, Vector2D=$Particles.SpriteSize)) {isSetParameters=true}
   /SetVariables(Particles.Glow Color=&MultiplyLinearColor(Linear Color=$Particles.Color, Scale Factor=<2 2 2 0.15>)) {isSetParameters=true}
   &KillParticles(Kill Particles=/Niagara/DynamicInputs/Helpers/MaskBool(
     Mask Source:NiagaraInt32=1
     Mask Parameter:NiagaraInt32=$Particles.PartitionParticles.Partition
     Mask:NiagaraInt32=0
     Masked Bool:NiagaraBool=$User.bisHandHidden
     Passthrough Bool:NiagaraBool=&MaskBool(
      Mask Source:NiagaraInt32=1
      Mask Parameter:NiagaraInt32=$Particles.PartitionParticles.Partition
      Mask:NiagaraInt32=1
      Masked Bool:NiagaraBool=$User.bisHandHidden
      Passthrough Bool:NiagaraBool=&MaskBool(
       Mask Source:NiagaraInt32=1
       Mask Parameter:NiagaraInt32=$Particles.PartitionParticles.Partition
       Mask:NiagaraInt32=2
       Masked Bool:NiagaraBool=&Boolean_OrOperation(Bool1=$User.bIsFootHidden, Bool2=$User.HideFootFX)
       Passthrough Bool:NiagaraBool=&MaskBool(
        Mask Source:NiagaraInt32=1
        Mask Parameter:NiagaraInt32=$Particles.PartitionParticles.Partition
        Mask:NiagaraInt32=3
        Masked Bool:NiagaraBool=&Boolean_OrOperation(Bool1=$User.bIsFootHidden, Bool2=$User.HideFootFX)
        Passthrough Bool:NiagaraBool=true
       )
      )
     )
    ))
  ]
  rendererClasses [&NiagaraSpriteRendererProperties]
  renderers [
   {rendererClass=&NiagaraSpriteRendererProperties}
   {rendererClass=&NiagaraSpriteRendererProperties, rendererIndex=1}
  ]
 }
 Mist {
  emitterUpdateScript [
   &EmitterState(Life Cycle Mode=::NewEnumerator1, Loop Duration=1, UseLoopDelay=false)
   &SpawnBurst_Instantaneous(Use Spawn Probability=false, UseLoopCountLimit=true) {enabled=false}
   &SpawnRate(SpawnRate=15, Use Spawn Probability=false)
  ]
  particleSpawnScript [
   &PartitionParticles(Number Of Partitions=4)
   &InitializeParticle(
    Color=&MergeHelper_ColorSwitch(
     0=<rgba 0.5625 0.7997 1 0.5>
     1=<rgba 0.5625 0.7997 1 0.5>
     2=<rgba 0.8 0.9262 1 0.21>
     3=<rgba 0.8 0.9262 1 0.21>
     4=<rgba 0.8 0.9262 1 0.21>
     Index=$Engine.QualityLevel
     Num=::NewEnumerator3
    )
    Lifetime=&RandomRangeFloat(Minimum=2.5, Maximum=3)
    Sprite Rotation Mode=::NewEnumerator2
    Sprite Size Mode=::NewEnumerator3
    Sprite UV Mode=::NewEnumerator3
    Uniform Sprite Size=&Multiply_Float(
     A=22
     B=&MergeHelper_FloatSwitch(0=1.8, 1=1.8, 2=1, 3=1, 4=1, Index=$Engine.QualityLevel, Num=::NewEnumerator3)
    )
    UsePositionOffset=false
    Write Material Random=true
   )
   &CameraOffset(Camera Offset Amount=10)
  ]
  particleUpdateScript [
   &ParticleState()
   &SkeletalMeshLocation(
    Bone / Socket Index=$Particles.PartitionParticles.Partition
    Bone Sampling Mode=::NewEnumerator4
    "Inherit Velocity (Scale)"=0
    MaskSpawnGroup=false
    Sampled Position Offset=&MergeHelper_VectorSwitch(
     0=&RandomRangeVector(Minimum=<0 0 0>, Maximum=<25 0 0>)
     1=&RandomRangeVector(Minimum=<0 0 0>, Maximum=<-25 0 0>)
     2=&RandomRangeVector(Minimum=<0 0 0>, Maximum=<30 0 5>)
     3=&RandomRangeVector(Minimum=<0 0 0>, Maximum=<-30 -6 -5>)
     Index=$Particles.PartitionParticles.Partition
     Num=::NewEnumerator2
    )
    Skeletal Mesh=$System.Skeletal Mesh Reader
    Velocity Sampling=::NewEnumerator0
    VelocitySpeedLimit=false
   )
   &ScaleSpriteSize(
    Uniform Curve Scale=1.1
    Uniform Curve Sprite Scale=DataInterface() {propertyValues {Curve {keys [<at 0 0.5 RCIM_Cubic>, {tangentMode=RCTM_User, time=1, value=1}]}}}
   )
   &ScaleColor(
    Scale Alpha=&FloatFromCurve(
     FloatCurve=DataInterface() {propertyValues {Curve {keys [
         {arriveTangent=0.0587, interpMode=RCIM_Cubic, leaveTangent=0.0587, tangentMode=RCTM_User, time=0, value=0}
         <at 0.5 1 RCIM_Cubic>
         {arriveTangent=0.0143, interpMode=RCIM_Cubic, leaveTangent=0.0143, tangentMode=RCTM_User, time=1, value=0}
        ]}}}
     Scale Curve=&Multiply_Float(A=0.35, B=&FN_FrontEndFloatModifier(Gameplay=1, Front End=0.9))
    )
    ScaleA=true
    ScaleRGB=true
   )
   &ScaleColor(
    Color Value To Scale=$Particles.Color
    Scale RGB=&MakeVectorFromLinearColorRGB(Color=&FNTimeOfDayColor(
      Night Color=<rgba 0.539 0.6941 1 1>
      Morning Color=<rgba 0.8743 0.9172 1 1>
      Day Color=<rgba 1 1 1 1>
      Evening Color=<rgba 0.7198 0.9904 1 1>
     ))
    ScaleA=true
    ScaleRGB=true
   ) {module=ScaleColor001}
   /SetVariables(Particles.SubImageIndex=&FloatFromCurve(FloatCurve=DataInterface() {propertyValues {Curve {keys [<at 0 30>,<at 1 64>]}}})) {isSetParameters=true}
   &DynamicMaterialParameters(
    Index 0 Param 1=3
    Index 0 Param 2=1
    Index 0 Param 3=0.5
    Module.Index0Float0Write=true
    Module.Index0Float1Write=true
    Module.Index0Float2Write=true
    Module.Index0Float3Write=true
   )
   &KillParticles(Kill Particles=&MaskBool(
     Mask Source:NiagaraInt32=1
     Mask Parameter:NiagaraInt32=$Particles.PartitionParticles.Partition
     Mask:NiagaraInt32=0
     Masked Bool:NiagaraBool=$User.bisHandHidden
     Passthrough Bool:NiagaraBool=&MaskBool(
      Mask Source:NiagaraInt32=1
      Mask Parameter:NiagaraInt32=$Particles.PartitionParticles.Partition
      Mask:NiagaraInt32=1
      Masked Bool:NiagaraBool=$User.bisHandHidden
      Passthrough Bool:NiagaraBool=&MaskBool(
       Mask Source:NiagaraInt32=1
       Mask Parameter:NiagaraInt32=$Particles.PartitionParticles.Partition
       Mask:NiagaraInt32=2
       Masked Bool:NiagaraBool=&Boolean_OrOperation(Bool1=$User.bIsFootHidden, Bool2=$User.HideFootFX)
       Passthrough Bool:NiagaraBool=&MaskBool(
        Mask Source:NiagaraInt32=1
        Mask Parameter:NiagaraInt32=$Particles.PartitionParticles.Partition
        Mask:NiagaraInt32=3
        Masked Bool:NiagaraBool=&Boolean_OrOperation(Bool1=$User.bIsFootHidden, Bool2=$User.HideFootFX)
        Passthrough Bool:NiagaraBool=true
       )
      )
     )
    ))
  ]
  rendererClasses [&NiagaraSpriteRendererProperties]
  renderers [
   {rendererClass=&NiagaraSpriteRendererProperties}
   {rendererClass=&NiagaraSpriteRendererProperties, rendererIndex=1}
  ]
 }
}
summary {
 systemName=NS_BriskReign_BodyMist
 emitters [
  {emitterName=Mist, rendererClasses [&NiagaraSpriteRendererProperties]}
  {emitterName=Embers, rendererClasses [&NiagaraSpriteRendererProperties]}
 ]
 userVariables [
  User.bIsFootHidden {type=NiagaraBool}
  User.bIsFrontEnd {type=NiagaraBool}
  User.bIsFrontEndPreview {type=NiagaraBool}
  User.bisHandHidden {type=NiagaraBool}
  User.bisHeadHidden {type=NiagaraBool}
  User.HideFootFX {type=NiagaraBool}
 ]
}
