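# Physics asset for the Belt mesh of the male medium BriskReign cosmetic body, setting up physics for two front straps and two jacket pieces.
# Each strap is a five-link capsule chain with tight twist and swing that widens down the chain, hung from a base body on spine_04.
# Each jacket piece is one capsule swinging mostly on one axis, hung from a base on spine_01.
# Base joints rotate freely, and unconstrained capsules cover pelvis, thighs and upper arms.
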PhysicsAsset /BRCosmetics/Characters/Player/Male/Medium/Bodies/M_MED_BriskReign/Meshes/M_MED_BriskReign_Belt_Physics
bodies {
 base_jacket_l [{shapeType=Capsule, length=0.505, radius=0.505, shapeName=base_jacket_l_capsule, rotation=<rot 0 -90 0>}]
 base_jacket_r [{shapeType=Capsule, length=0.505, radius=0.505, shapeName=base_jacket_r_capsule, rotation=<rot 0 -90 0>}]
 base_strap_fr_l [{shapeType=Capsule, length=1, radius=1, shapeName=base_strap_fr_l_capsule, rotation=<rot 0 -90 0>}]
 base_strap_fr_r [{shapeType=Capsule, length=1, radius=1, shapeName=base_strap_fr_r_capsule, rotation=<rot 0 -90 0>}]
 dyn_jacket_l [{shapeType=Capsule, length=6, radius=3.5, shapeName=dyn_jacket_r_capsule, center=<16 0 5>, rotation=<rot -80 89.998 0.0019>}]
 dyn_jacket_r [{shapeType=Capsule, length=6, radius=3.5, shapeName=dyn_jacket_r_capsule, center=<16 0 5>, rotation=<rot -75 -89.9986 179.9983>}]
 dyn_strap_fr_01_l [{shapeType=Capsule, length=5, radius=2.5, shapeName=dyn_strap_fr_01_l_capsule, center=<4 0 2.0618>, rotation=<rot -90 0 0>}]
 dyn_strap_fr_01_r [{shapeType=Capsule, length=5, radius=2.5, shapeName=dyn_strap_fr_01_r_capsule, center=<4 0 2>, rotation=<rot -90 0 0>}]
 dyn_strap_fr_02_l [{shapeType=Capsule, length=5, radius=2.5, shapeName=dyn_strap_fr_02_l_capsule, center=<4 0 2.0618>, rotation=<rot -90 0 0>}]
 dyn_strap_fr_02_r [{shapeType=Capsule, length=5, radius=2.5, shapeName=dyn_strap_fr_02_r_capsule, center=<4 0 2>, rotation=<rot -90 0 0>}]
 dyn_strap_fr_03_l [{shapeType=Capsule, length=5, radius=2.5, shapeName=dyn_strap_fr_03_l_capsule, center=<4 0 2.0618>, rotation=<rot -90 0 0>}]
 dyn_strap_fr_03_r [{shapeType=Capsule, length=5, radius=2.5, shapeName=dyn_strap_fr_03_r_capsule, center=<4 0 2.0618>, rotation=<rot -90 0 0>}]
 dyn_strap_fr_04_l [{shapeType=Capsule, length=5, radius=2.5, shapeName=dyn_strap_fr_04_l_capsule, center=<4 0 2>, rotation=<rot -90 0 0>}]
 dyn_strap_fr_04_r [{shapeType=Capsule, length=5, radius=2.5, shapeName=dyn_strap_fr_04_r_capsule, center=<4 0 2>, rotation=<rot -90 0 0>}]
 dyn_strap_fr_05_l [{shapeType=Capsule, length=5, radius=2.5, shapeName=dyn_strap_fr_05_l_capsule, center=<4 0 2>, rotation=<rot -90 0 0>}]
 dyn_strap_fr_05_r [{shapeType=Capsule, length=5, radius=2.5, shapeName=dyn_strap_fr_05_r_capsule, center=<4 0 2>, rotation=<rot -90 0 0>}]
 pelvis [{shapeType=Capsule, length=11.2204, radius=11.2204, shapeName=pelvis_capsule}]
 spine_01 [{shapeType=Capsule, length=10.9238, radius=10.9238, shapeName=spine_01_capsule, rotation=<rot 0 -0.0001 0>}]
 spine_02 [{shapeType=Capsule, length=10.9238, radius=10.9238, shapeName=spine_02_capsule}]
 spine_03 [{shapeType=Capsule, length=10.9238, radius=10.9238, shapeName=spine_03_capsule, rotation=<rot 0 -0.0003 180>}]
 spine_04 [{shapeType=Capsule, length=10.9238, radius=10.9238, shapeName=spine_04_capsule, center=<0 1.4849 0>}]
 spine_05 [{shapeType=Capsule, length=10.9238, radius=10.9238, shapeName=spine_05_capsule, center=<-4.8576 2.6903 0>}]
 thigh_l [{shapeType=Capsule, length=27.0643, radius=11.0643, shapeName=thigh_l_capsule, center=<-20 0 0>, rotation=<rot 0 90 -90>}]
 thigh_r [{shapeType=Capsule, length=27.0643, radius=11.0643, shapeName=thigh_r_capsule, center=<20 0 0>, rotation=<rot 0 90 -90>}]
 upperarm_l [{shapeType=Capsule, length=12, radius=2.5, shapeName=upperarm_l_capsule, center=<12 -0.83 0>, rotation=<rot -90 0 0>}]
 upperarm_r [{shapeType=Capsule, length=12, radius=6, shapeName=upperarm_r_capsule, center=<-12 0 0>, rotation=<rot -90 0 0>}]
}
constraints [
 {
  bone1Name=base_strap_fr_l
  bone2Name=spine_04
  swing1LimitDegrees=45
  swing1Motion=Free
  swing2LimitDegrees=45
  swing2Motion=Free
  twistLimitDegrees=45
  twistMotion=Free
 }
 {bone1Name=dyn_strap_fr_01_l, bone2Name=base_strap_fr_l, swing1LimitDegrees=25, swing2LimitDegrees=10, twistLimitDegrees=1}
 {bone1Name=dyn_strap_fr_02_l, bone2Name=dyn_strap_fr_01_l, swing1LimitDegrees=25, swing2LimitDegrees=15, twistLimitDegrees=1}
 {bone1Name=dyn_strap_fr_03_l, bone2Name=dyn_strap_fr_02_l, swing1LimitDegrees=25, swing2LimitDegrees=25, twistLimitDegrees=1}
 {bone1Name=dyn_strap_fr_04_l, bone2Name=dyn_strap_fr_03_l, swing1LimitDegrees=25, swing2LimitDegrees=25, twistLimitDegrees=1}
 {bone1Name=dyn_strap_fr_05_l, bone2Name=dyn_strap_fr_04_l, swing1LimitDegrees=25, swing2LimitDegrees=25, twistLimitDegrees=1}
 {
  bone1Name=base_strap_fr_r
  bone2Name=spine_04
  swing1LimitDegrees=45
  swing1Motion=Free
  swing2LimitDegrees=45
  swing2Motion=Free
  twistLimitDegrees=45
  twistMotion=Free
 }
 {bone1Name=dyn_strap_fr_01_r, bone2Name=base_strap_fr_r, swing1LimitDegrees=25, swing2LimitDegrees=10, twistLimitDegrees=1}
 {bone1Name=dyn_strap_fr_02_r, bone2Name=dyn_strap_fr_01_r, swing1LimitDegrees=25, swing2LimitDegrees=15, twistLimitDegrees=1}
 {bone1Name=dyn_strap_fr_03_r, bone2Name=dyn_strap_fr_02_r, swing1LimitDegrees=25, swing2LimitDegrees=25, twistLimitDegrees=1}
 {bone1Name=dyn_strap_fr_04_r, bone2Name=dyn_strap_fr_03_r, swing1LimitDegrees=25, swing2LimitDegrees=25, twistLimitDegrees=1}
 {bone1Name=dyn_strap_fr_05_r, bone2Name=dyn_strap_fr_04_r, swing1LimitDegrees=25, swing2LimitDegrees=25, twistLimitDegrees=1}
 {
  bone1Name=base_jacket_l
  bone2Name=spine_01
  swing1LimitDegrees=45
  swing1Motion=Free
  swing2LimitDegrees=45
  swing2Motion=Free
  twistLimitDegrees=45
  twistMotion=Free
 }
 {bone1Name=dyn_jacket_l, bone2Name=base_jacket_l, swing1LimitDegrees=10, swing2LimitDegrees=75, twistLimitDegrees=5}
 {
  bone1Name=base_jacket_r
  bone2Name=spine_01
  swing1LimitDegrees=45
  swing1Motion=Free
  swing2LimitDegrees=45
  swing2Motion=Free
  twistLimitDegrees=45
  twistMotion=Free
 }
 {bone1Name=dyn_jacket_r, bone2Name=base_jacket_r, swing1LimitDegrees=10, swing2LimitDegrees=75, twistLimitDegrees=5}
]
