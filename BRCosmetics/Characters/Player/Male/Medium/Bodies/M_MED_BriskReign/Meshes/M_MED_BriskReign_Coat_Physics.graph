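# Physics asset for the BriskReign male medium body's coat, setting up the back of the coat for physical movement.
# Center, left and right five-bone capsule chains hang from pelvis-linked bases with tight twist and limited swing, and free-rotation links tie the side chains to the center at the third bone.
# Pelvis, thigh, calf and foot capsules act as body colliders, while a huge root sphere offset downward approximates a floor.
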PhysicsAsset /BRCosmetics/Characters/Player/Male/Medium/Bodies/M_MED_BriskReign/Meshes/M_MED_BriskReign_Coat_Physics
bodies {
 base_coat_bk [{shapeType=Capsule, length=0.505, radius=0.505, shapeName=base_coat_bk_capsule, rotation=<rot 0 -90 0>}]
 base_coat_bk_l [{shapeType=Capsule, length=0.505, radius=0.505, shapeName=base_coat_bk_l_capsule, rotation=<rot 0 -90 0>}]
 base_coat_bk_r [{shapeType=Capsule, length=0.505, radius=0.505, shapeName=base_coat_bk_r_capsule, rotation=<rot 0 -90 0>}]
 calf_l [{shapeType=Capsule, length=39.0313, radius=8.0313, shapeName=calf_l_capsule, center=<-15 -3 0>, rotation=<rot 0 90 -90>}]
 calf_r [{shapeType=Capsule, length=39.0313, radius=8.0313, shapeName=calf_r_capsule, center=<15 3 0>, rotation=<rot 0 90 -90>}]
 dyn_coat_bk_01 [{shapeType=Capsule, length=12, radius=7, shapeName=dyn_coat_bk_01_capsule, center=<7 0 5>, rotation=<rot 0 0 90>}]
 dyn_coat_bk_01_l [{shapeType=Capsule, length=12, radius=7, shapeName=dyn_coat_bk_01_l_capsule, center=<7 0 6>, rotation=<rot 0 0 90>}]
 dyn_coat_bk_01_r [{shapeType=Capsule, length=12, radius=7, shapeName=dyn_coat_bk_01_r_capsule, center=<7 0 6>, rotation=<rot 0 0 90>}]
 dyn_coat_bk_02 [{shapeType=Capsule, length=12, radius=10, shapeName=dyn_coat_bk_02_capsule, center=<7 0 3>, rotation=<rot 0 0 90>}]
 dyn_coat_bk_02_l [{shapeType=Capsule, length=12, radius=10, shapeName=dyn_coat_bk_02_l_capsule, center=<7 0 5>, rotation=<rot 0 0 90>}]
 dyn_coat_bk_02_r [{shapeType=Capsule, length=12, radius=10, shapeName=dyn_coat_bk_02_r_capsule, center=<6.0026 0 5>, rotation=<rot 0 0 90>}]
 dyn_coat_bk_03 [{shapeType=Capsule, length=12, radius=8, shapeName=dyn_coat_bk_03_capsule, center=<6 0 1>, rotation=<rot 0 0 90>}]
 dyn_coat_bk_03_l [{shapeType=Capsule, length=12, radius=8, shapeName=dyn_coat_bk_03_l_capsule, center=<7 0 4>, rotation=<rot 0 0 90>}]
 dyn_coat_bk_03_r [{shapeType=Capsule, length=12, radius=8, shapeName=dyn_coat_bk_03_r_capsule, center=<7 0 4>, rotation=<rot 0 0 90>}]
 dyn_coat_bk_04 [{shapeType=Capsule, length=12, radius=8, shapeName=dyn_coat_bk_04_capsule, center=<7 0 0>, rotation=<rot 0 0 90>}]
 dyn_coat_bk_04_l [{shapeType=Capsule, length=12, radius=8, shapeName=dyn_coat_bk_04_l_capsule, center=<7 0 3>, rotation=<rot 0 0 90>}]
 dyn_coat_bk_04_r [{shapeType=Capsule, length=12, radius=8, shapeName=dyn_coat_bk_04_r_capsule, center=<7 0 3>, rotation=<rot 0 0 90>}]
 dyn_coat_bk_05 [{shapeType=Capsule, length=12, radius=8, shapeName=dyn_coat_bk_05_capsule, center=<7 0 0>, rotation=<rot 0 0 90>}]
 dyn_coat_bk_05_l [{shapeType=Capsule, length=12, radius=8, shapeName=dyn_coat_bk_05_l_capsule, center=<7 0 3>, rotation=<rot 0 0 90>}]
 dyn_coat_bk_05_r [{shapeType=Capsule, length=12, radius=8, shapeName=dyn_coat_bk_05_r_capsule, center=<7 0 3>, rotation=<rot 0 0 90>}]
 foot_l [{shapeType=Capsule, length=11.0044, radius=11.0044, shapeName=foot_l_capsule, center=<-5 8 0>, rotation=<rot 0 0 -90>}]
 foot_r [{shapeType=Capsule, length=11.0044, radius=11.0044, shapeName=foot_r_capsule, center=<5 -8 0>, rotation=<rot 0 0 -90>}]
 pelvis [{shapeType=Capsule, length=10.861, radius=10.861, shapeName=pelvis_capsule, center=<0 -1 0>, rotation=<rot 0 90 0>}]
 root [{shapeType=Sphere, radius=500, shapeName=root_sphere, center=<0 17.4691 -497.1>}]
 thigh_l [{shapeType=Capsule, length=27.0643, radius=11.0643, shapeName=thigh_l_capsule, center=<-20 0 0>, rotation=<rot 0 90 -90>}]
 thigh_r [{shapeType=Capsule, length=27.0643, radius=11.0643, shapeName=thigh_r_capsule, center=<20 0 0>, rotation=<rot 0 90 -90>}]
}
constraints [
 {
  bone1Name=base_coat_bk
  bone2Name=pelvis
  swing1LimitDegrees=45
  swing1Motion=Free
  swing2LimitDegrees=45
  swing2Motion=Free
  twistLimitDegrees=45
  twistMotion=Free
 }
 {bone1Name=dyn_coat_bk_01, bone2Name=base_coat_bk, swing1LimitDegrees=15, swing2LimitDegrees=25, twistLimitDegrees=5}
 {bone1Name=dyn_coat_bk_02, bone2Name=dyn_coat_bk_01, swing1LimitDegrees=15, swing2LimitDegrees=30, twistLimitDegrees=5}
 {bone1Name=dyn_coat_bk_03, bone2Name=dyn_coat_bk_02, swing1LimitDegrees=15, swing2LimitDegrees=30, twistLimitDegrees=5}
 {bone1Name=dyn_coat_bk_04, bone2Name=dyn_coat_bk_03, swing1LimitDegrees=15, swing2LimitDegrees=30, twistLimitDegrees=5}
 {bone1Name=dyn_coat_bk_05, bone2Name=dyn_coat_bk_04, swing1LimitDegrees=15, swing2LimitDegrees=30, twistLimitDegrees=5}
 {
  bone1Name=base_coat_bk_l
  bone2Name=pelvis
  swing1LimitDegrees=45
  swing1Motion=Free
  swing2LimitDegrees=45
  swing2Motion=Free
  twistLimitDegrees=45
  twistMotion=Free
 }
 {bone1Name=dyn_coat_bk_01_l, bone2Name=base_coat_bk_l, swing1LimitDegrees=15, swing2LimitDegrees=25, twistLimitDegrees=5}
 {bone1Name=dyn_coat_bk_02_l, bone2Name=dyn_coat_bk_01_l, swing1LimitDegrees=15, swing2LimitDegrees=30, twistLimitDegrees=5}
 {bone1Name=dyn_coat_bk_03_l, bone2Name=dyn_coat_bk_02_l, swing1LimitDegrees=15, swing2LimitDegrees=30, twistLimitDegrees=5}
 {bone1Name=dyn_coat_bk_04_l, bone2Name=dyn_coat_bk_03_l, swing1LimitDegrees=15, swing2LimitDegrees=30, twistLimitDegrees=5}
 {bone1Name=dyn_coat_bk_05_l, bone2Name=dyn_coat_bk_04_l, swing1LimitDegrees=15, swing2LimitDegrees=30, twistLimitDegrees=5}
 {
  bone1Name=base_coat_bk_r
  bone2Name=pelvis
  swing1LimitDegrees=45
  swing1Motion=Free
  swing2LimitDegrees=45
  swing2Motion=Free
  twistLimitDegrees=45
  twistMotion=Free
 }
 {bone1Name=dyn_coat_bk_01_r, bone2Name=base_coat_bk_r, swing1LimitDegrees=15, swing2LimitDegrees=25, twistLimitDegrees=5}
 {bone1Name=dyn_coat_bk_02_r, bone2Name=dyn_coat_bk_01_r, swing1LimitDegrees=15, swing2LimitDegrees=30, twistLimitDegrees=5}
 {bone1Name=dyn_coat_bk_03_r, bone2Name=dyn_coat_bk_02_r, swing1LimitDegrees=15, swing2LimitDegrees=30, twistLimitDegrees=5}
 {bone1Name=dyn_coat_bk_04_r, bone2Name=dyn_coat_bk_03_r, swing1LimitDegrees=15, swing2LimitDegrees=30, twistLimitDegrees=5}
 {bone1Name=dyn_coat_bk_05_r, bone2Name=dyn_coat_bk_04_r, swing1LimitDegrees=15, swing2LimitDegrees=30, twistLimitDegrees=5}
 {
  bone1Name=dyn_coat_bk_03_l
  bone2Name=dyn_coat_bk_03
  swing1LimitDegrees=45
  swing1Motion=Free
  swing2LimitDegrees=45
  swing2Motion=Free
  twistLimitDegrees=45
  twistMotion=Free
 }
 {
  bone1Name=dyn_coat_bk_03_r
  bone2Name=dyn_coat_bk_03
  swing1LimitDegrees=45
  swing1Motion=Free
  swing2LimitDegrees=45
  swing2Motion=Free
  twistLimitDegrees=45
  twistMotion=Free
 }
 {
  bone1Name=pelvis
  bone2Name=root
  swing1LimitDegrees=45
  swing1Motion=Free
  swing2LimitDegrees=45
  swing2Motion=Free
  twistLimitDegrees=45
  twistMotion=Free
 }
 {
  bone1Name=thigh_l
  bone2Name=root
  swing1LimitDegrees=45
  swing1Motion=Free
  swing2LimitDegrees=45
  swing2Motion=Free
  twistLimitDegrees=45
  twistMotion=Free
 }
 {
  bone1Name=calf_l
  bone2Name=root
  swing1LimitDegrees=45
  swing1Motion=Free
  swing2LimitDegrees=45
  swing2Motion=Free
  twistLimitDegrees=45
  twistMotion=Free
 }
 {
  bone1Name=foot_l
  bone2Name=root
  swing1LimitDegrees=45
  swing1Motion=Free
  swing2LimitDegrees=45
  swing2Motion=Free
  twistLimitDegrees=45
  twistMotion=Free
 }
 {
  bone1Name=thigh_r
  bone2Name=root
  swing1LimitDegrees=45
  swing1Motion=Free
  swing2LimitDegrees=45
  swing2Motion=Free
  twistLimitDegrees=45
  twistMotion=Free
 }
 {
  bone1Name=calf_r
  bone2Name=root
  swing1LimitDegrees=45
  swing1Motion=Free
  swing2LimitDegrees=45
  swing2Motion=Free
  twistLimitDegrees=45
  twistMotion=Free
 }
 {
  bone1Name=foot_r
  bone2Name=root
  swing1LimitDegrees=45
  swing1Motion=Free
  swing2LimitDegrees=45
  swing2Motion=Free
  twistLimitDegrees=45
  twistMotion=Free
 }
]
